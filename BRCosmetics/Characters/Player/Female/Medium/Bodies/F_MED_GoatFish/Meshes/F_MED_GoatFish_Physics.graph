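# Physics asset for the F_MED_GoatFish female medium player body, covering a left and right dangling string plus the upper spine.
# Each string is a chain of small capsules, a base and four dyn segments, with narrow swing limits and zero twist limit between links.
# Both bases attach to spine_05 with free angular motion.
# Larger spine capsules, three on spine_05, form a torso volume, likely for the strings to collide against.
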PhysicsAsset /BRCosmetics/Characters/Player/Female/Medium/Bodies/F_MED_GoatFish/Meshes/F_MED_GoatFish_Physics
bodies {
 base_string_l [{shapeType=Capsule, length=2, radius=1, shapeName=base_string_l_capsule, center=<3 0 0>, rotation=<rot 90 0 0>}]
 base_string_r [{shapeType=Capsule, length=2, radius=1, shapeName=base_string_r_capsule, center=<3 0 0>, rotation=<rot 90 0 0>}]
 dyn_string_1_l [{shapeType=Capsule, length=1.4239, radius=1.3761, shapeName=dyn_string_1_l_capsule, center=<3 0 0.918>, rotation=<rot 90 0 0>}]
 dyn_string_1_r [{shapeType=Capsule, length=1.4555, radius=1.3445, shapeName=dyn_string_1_r_capsule, center=<3 0 0.918>, rotation=<rot 90 0 0>}]
 dyn_string_2_l [{shapeType=Capsule, length=1.4239, radius=1.3761, shapeName=dyn_string_2_l_capsule, center=<3 0 0.818>, rotation=<rot 90 0 0>}]
 dyn_string_2_r [{shapeType=Capsule, length=1.4555, radius=1.3445, shapeName=dyn_string_2_r_capsule, center=<3 0 0.818>, rotation=<rot 90 0 0>}]
 dyn_string_3_l [{shapeType=Capsule, length=1.4471, radius=1.3529, shapeName=dyn_string_3_l_capsule, center=<3 0 0.8013>, rotation=<rot 90 0 0>}]
 dyn_string_3_r [{shapeType=Capsule, length=1.4787, radius=1.3213, shapeName=dyn_string_3_r_capsule, center=<3 0 0.8013>, rotation=<rot 90 0 0>}]
 dyn_string_4_l [{shapeType=Capsule, length=2.0608, radius=1.7392, shapeName=dyn_string_4_l_capsule, center=<2.8084 0 0.9601>, rotation=<rot 90 0 0>}]
 dyn_string_4_r [{shapeType=Capsule, length=2.0925, radius=1.7075, shapeName=dyn_string_4_r_capsule, center=<2.8084 0 0.9601>, rotation=<rot 90 0 0>}]
 spine_03 [{shapeType=Capsule, length=6.3731, radius=8.9771, shapeName=spine_03_capsule, center=<0 2.4309 0>, rotation=<rot 0 -90 0>}]
 spine_04 [{shapeType=Capsule, length=3.7109, radius=13.7139, shapeName=spine_04_capsule, center=<-2 0.6685 0>, rotation=<rot 0 -90 0>}]
 spine_05 [
  {
   shapeType=Capsule
   length=2.8432
   radius=17.333
   shapeName=spine_05_capsule
   center=<-8.0056 -4.8301 0.2299>
   rotation=<rot 0 1.3161 -179.3442>
  }
  {shapeType=Capsule, length=10, radius=10, shapeName=spine_05_capsule, center=<-3.2 1 5>, rotation=<rot -86 0 -40>}
  {shapeType=Capsule, length=10, radius=10, shapeName=spine_05_capsule, center=<-3.2 1 -5>, rotation=<rot -86 0 -40>}
 ]
}
constraints [
 {
  bone1Name=base_string_l
  bone2Name=spine_05
  swing1LimitDegrees=45
  swing1Motion=Free
  swing2LimitDegrees=45
  swing2Motion=Free
  twistLimitDegrees=45
  twistMotion=Free
 }
 {bone1Name=dyn_string_1_l, bone2Name=base_string_l, swing1LimitDegrees=10, swing2LimitDegrees=20, twistLimitDegrees=0}
 {bone1Name=dyn_string_2_l, bone2Name=dyn_string_1_l, swing1LimitDegrees=20, swing2LimitDegrees=20, twistLimitDegrees=0}
 {bone1Name=dyn_string_3_l, bone2Name=dyn_string_2_l, swing1LimitDegrees=20, swing2LimitDegrees=20, twistLimitDegrees=0}
 {bone1Name=dyn_string_4_l, bone2Name=dyn_string_3_l, swing1LimitDegrees=20, swing2LimitDegrees=20, twistLimitDegrees=0}
 {
  bone1Name=base_string_r
  bone2Name=spine_05
  swing1LimitDegrees=45
  swing1Motion=Free
  swing2LimitDegrees=45
  swing2Motion=Free
  twistLimitDegrees=45
  twistMotion=Free
 }
 {bone1Name=dyn_string_1_r, bone2Name=base_string_r, swing1LimitDegrees=10, swing2LimitDegrees=20, twistLimitDegrees=0}
 {bone1Name=dyn_string_2_r, bone2Name=dyn_string_1_r, swing1LimitDegrees=20, swing2LimitDegrees=20, twistLimitDegrees=0}
 {bone1Name=dyn_string_3_r, bone2Name=dyn_string_2_r, swing1LimitDegrees=20, swing2LimitDegrees=20, twistLimitDegrees=0}
 {bone1Name=dyn_string_4_r, bone2Name=dyn_string_3_r, swing1LimitDegrees=20, swing2LimitDegrees=20, twistLimitDegrees=0}
 {
  bone1Name=spine_04
  bone2Name=spine_03
  swing1LimitDegrees=45
  swing1Motion=Free
  swing2LimitDegrees=45
  swing2Motion=Free
  twistLimitDegrees=45
  twistMotion=Free
 }
 {
  bone1Name=spine_05
  bone2Name=spine_04
  swing1LimitDegrees=45
  swing1Motion=Free
  swing2LimitDegrees=45
  swing2Motion=Free
  twistLimitDegrees=45
  twistMotion=Free
 }
]
