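# Post-process-style AnimBlueprint for the GoatFish female body cosmetic, adding secondary motion to its hoodie, zipper, strings and hem.
# It copies the parent mesh pose through a Control Rig, then chains spring AnimDynamics, a weighted neck/head zipper constraint and RigidBody simulation.
# Additive bone offsets adjust hood and strings.
# The event graph derives gravity/RigidBody settings from a parameter asset, plus crouch-walk, front-end-idle and head-direction values measured against spine_05.
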AnimBlueprint /BRCosmetics/Characters/Player/Female/Medium/Bodies/F_MED_GoatFish/Meshes/F_MED_GoatFish_AnimBP
target_skeleton=&/F_MED_GoatFish_Skeleton
graphs [
 AnimGraph {nodes [
   "Transform (Modify) Bone - Bone: base_hoodie"(Alpha=1, Translation="8.474,0.000,-1.871", Rotation="-14.442,0.000,-0.000", Scale="1,1,1") {
    rotationMode=BMM_Additive
    rotationSpace=BCS_BoneSpace
    scaleMode=BMM_Additive
    scaleSpace=BCS_BoneSpace
    translationMode=BMM_Additive
    translationSpace=BCS_BoneSpace
    alphaBoolBlend {blendInTime=0.5, blendOutTime=0.5}
    alphaScaleBiasClamp {bInterpResult=true}
    boneToModify {boneName=base_hoodie}
    rotation=<rot -14.442 0 0>
    translation {x=8.474, z=-1.871}
   }
   n1 = Control Rig(Source=Copy Pose From Mesh() {bUseAttachedParent=true}) {controlRigAssetReference {blueprintRigClass=&/F_MED_GoatFish_CtrlRig.F_MED_GoatFish_CtrlRig_C}}
   n2 = Anim Dynamics - Bone: dyn_hoodie!(ComponentPose=Local To Component(LocalPose=@n1), Alpha=1, GravityOverride:Vector="0,0,-980", ExternalForce="0,0,0") -> (Pose) {
    angularDampingOverride=1
    angularSpringConstant=20000
    bAngularSpring=true
    bOverrideAngularDamping=true
    bUseGravityOverride=true
    gravityScale=0.35
    simulationSpace=RootRelative
    boundBone {boneName=dyn_hoodie}
    componentAppliedLinearAccClamp=<5000 5000 5000>
    componentLinearAccScale=<2 2 2>
    componentLinearVelScale=<2 2 2>
    gravityOverride {z=-980}
    physicsBodyDefinitions [{
      boundBone {boneName=dyn_hoodie}
      boxExtents=<20 30 30>
      constraintSetup {angularLimitsMax=<0 0 0>, angularLimitsMin=<0 -30 0>, angularTarget=<1 0 0>}
      localJointOffset=<4 0 0>
     }]
   }
   n3 = Anim Dynamics - Bone: dyn_zipper_1!(ComponentPose=@n2, Alpha=1, GravityOverride:Vector="0,0,-980", ExternalForce="0,0,0") -> (Pose) {
    angularDampingOverride=1
    angularSpringConstant=20000
    bAngularSpring=true
    bOverrideAngularDamping=true
    bUseGravityOverride=true
    gravityScale=0.35
    simulationSpace=RootRelative
    boundBone {boneName=dyn_zipper_1}
    componentAppliedLinearAccClamp=<3000 3000 3000>
    componentLinearAccScale {y=2, z=2}
    componentLinearVelScale {z=2}
    gravityOverride {z=-980}
    physicsBodyDefinitions [{
      boundBone {boneName=dyn_zipper_1}
      boxExtents=<10 20 20>
      constraintSetup {angularLimitsMax=<0 0 50>, angularLimitsMin=<0 -90 -50>, angularTarget=<1 0 0>}
      localJointOffset=<4 0 0>
     }]
   }
   n4 = "Constraint - base_zipper (PPPPPPPP)"!(
    ComponentPose=@n3
    Alpha=1
    ConstraintWeights_0=0.3456
    ConstraintWeights_1:float=0.1789
    ConstraintWeights_2:float=0.1546
    ConstraintWeights_3:float=0.1179
    ConstraintWeights_4:float=0.0939
    ConstraintWeights_5:float=0.0841
    ConstraintWeights_6:float=0.0138
    ConstraintWeights_7:float=0.0112
   ) -> (Pose) {
    boneToModify {boneName=base_zipper}
    constraintSetup [
     {targetBone {boneName=neck_02}}
     {targetBone {boneName=FACIAL_C_Jawline}}
     {targetBone {boneName=neck_01}}
     {targetBone {boneName=head}}
     {targetBone {boneName=spine_05}}
     {targetBone {boneName=FACIAL_C_UnderChin}}
     {targetBone {boneName=FACIAL_R_UnderChin}}
     {targetBone {boneName=FACIAL_L_UnderChin}}
    ]
    constraintWeights [1,1,1,1,1,1,1,1]
   }
   n5 = Anim Dynamics - Bone: dyn_bottom_1_l!(ComponentPose=@n4, Alpha=1, GravityOverride:Vector="0,0,-980", ExternalForce="0,0,0") -> (Pose) {
    angularDampingOverride=1
    angularSpringConstant=2000
    bAngularSpring=true
    bOverrideAngularDamping=true
    bUseGravityOverride=true
    gravityScale=0.35
    simulationSpace=RootRelative
    boundBone {boneName=dyn_bottom_1_l}
    componentLinearAccScale=<4 4 4>
    componentLinearVelScale=<4 4 4>
    gravityOverride {z=-980}
    physicsBodyDefinitions [{
      boundBone {boneName=dyn_bottom_1_l}
      boxExtents=<20 30 30>
      constraintSetup {angularLimitsMax=<50 50 15>, angularLimitsMin=<-50 -50 -15>, angularTarget=<1 0 0>}
      localJointOffset=<0 0 -10>
     }]
   }
   n6 = Anim Dynamics - Bone: dyn_bottom_1_r!(ComponentPose=@n5, Alpha=1, GravityOverride:Vector="0,0,-980", ExternalForce="0,0,0") -> (Pose) {
    angularDampingOverride=1
    angularSpringConstant=2000
    bAngularSpring=true
    bOverrideAngularDamping=true
    bUseGravityOverride=true
    gravityScale=0.35
    simulationSpace=RootRelative
    boundBone {boneName=dyn_bottom_1_r}
    componentLinearAccScale=<4 4 4>
    componentLinearVelScale=<4 4 4>
    gravityOverride {z=-980}
    physicsBodyDefinitions [{
      boundBone {boneName=dyn_bottom_1_r}
      boxExtents=<20 30 30>
      constraintSetup {angularLimitsMax=<50 50 15>, angularLimitsMin=<-50 -50 -15>, angularTarget=<1 0 0>}
      localJointOffset=<0 0 -10>
     }]
   }
   n7 = Anim Dynamics - Bone: dyn_bottom_2_l!(ComponentPose=@n6, Alpha=1, GravityOverride:Vector="0,0,-980", ExternalForce="0,0,0") -> (Pose) {
    angularDampingOverride=1
    angularSpringConstant=2000
    bAngularSpring=true
    bOverrideAngularDamping=true
    bUseGravityOverride=true
    gravityScale=0.35
    simulationSpace=RootRelative
    boundBone {boneName=dyn_bottom_2_l}
    componentLinearAccScale=<4 4 4>
    componentLinearVelScale=<4 4 4>
    gravityOverride {z=-980}
    physicsBodyDefinitions [{
      boundBone {boneName=dyn_bottom_2_l}
      boxExtents=<20 30 30>
      constraintSetup {angularLimitsMax=<50 50 15>, angularLimitsMin=<-50 -50 -15>, angularTarget=<1 0 0>}
      localJointOffset=<0 0 -10>
     }]
   }
   n8 = Anim Dynamics - Bone: dyn_bottom_2_r!(ComponentPose=@n7, Alpha=1, GravityOverride:Vector="0,0,-980", ExternalForce="0,0,0") -> (Pose) {
    angularDampingOverride=1
    angularSpringConstant=2000
    bAngularSpring=true
    bOverrideAngularDamping=true
    bUseGravityOverride=true
    gravityScale=0.35
    simulationSpace=RootRelative
    boundBone {boneName=dyn_bottom_2_r}
    componentLinearAccScale=<4 4 4>
    componentLinearVelScale=<4 4 4>
    gravityOverride {z=-980}
    physicsBodyDefinitions [{
      boundBone {boneName=dyn_bottom_2_r}
      boxExtents=<20 30 30>
      constraintSetup {angularLimitsMax=<50 50 15>, angularLimitsMin=<-50 -50 -15>, angularTarget=<1 0 0>}
      localJointOffset=<0 0 -10>
     }]
   }
   n9 = "Transform (Modify) Bone - Bone: dyn_hoodie"(ComponentPose=@n8, Alpha=1, Translation="7.916,-0.000,0.000", Rotation="0,0,0", Scale="1,1,1") {
    rotationMode=BMM_Additive
    rotationSpace=BCS_BoneSpace
    scaleMode=BMM_Additive
    scaleSpace=BCS_BoneSpace
    translationMode=BMM_Additive
    translationSpace=BCS_BoneSpace
    alphaScaleBiasClamp {bInterpResult=true, interpSpeedDecreasing=20, interpSpeedIncreasing=20}
    boneToModify {boneName=dyn_hoodie}
    translation {x=7.916}
   }
   n10 = "Transform (Modify) Bone - Bone: dyn_hoodie"(ComponentPose=@n9, Alpha=1, Translation="0,0,0", Rotation="6.438,0.194,-0.073", Scale="1,1,1") {
    rotationMode=BMM_Additive
    rotationSpace=BCS_BoneSpace
    scaleMode=BMM_Additive
    scaleSpace=BCS_BoneSpace
    translationMode=BMM_Additive
    translationSpace=BCS_BoneSpace
    alphaScaleBiasClamp {bInterpResult=true, interpSpeedDecreasing=20, interpSpeedIncreasing=20}
    boneToModify {boneName=dyn_hoodie}
    rotation=<rot 6.438 0.194 -0.073>
   }
   n11 = RigidBody!(ComponentPose=@n10, Alpha, OverrideWorldGravity:Vector="0,0,0", ExternalForce) -> (Pose) {
    bOverrideWorldGravity=true
    overridePhysicsAsset=&/F_MED_GoatFish_Physics
    simulationSpace=BaseBoneSpace
    baseBoneRef {boneName=spine_05}
    simSpaceSettings {velocityScaleZ=0.5, worldAlpha=1}
   }
   n12 = "Transform (Modify) Bone - Bone: dyn_string_1_r"!(ComponentPose=@n11, bAlphaBoolEnabled:bool=false, Translation="0,0,0", Rotation="16.888,0.038,0.036", Scale="1,1,1") -> (Pose) {
    alphaInputType=Bool
    rotationMode=BMM_Additive
    rotationSpace=BCS_BoneSpace
    scaleMode=BMM_Additive
    scaleSpace=BCS_BoneSpace
    translationMode=BMM_Additive
    translationSpace=BCS_BoneSpace
    alphaBoolBlend {blendInTime=0.5, blendOutTime=0.5}
    alphaScaleBiasClamp {bInterpResult=true, interpSpeedDecreasing=20, interpSpeedIncreasing=20}
    boneToModify {boneName=dyn_string_1_r}
    rotation=<rot 16.8881 0.038 0.036>
   }
   n13 = "Transform (Modify) Bone - Bone: dyn_string_1_l"!(ComponentPose=@n12, bAlphaBoolEnabled:bool=false, Translation="0,0,0", Rotation="12.023,0.004,0.016", Scale="1,1,1") -> (Pose) {
    alphaInputType=Bool
    rotationMode=BMM_Additive
    rotationSpace=BCS_BoneSpace
    scaleMode=BMM_Additive
    scaleSpace=BCS_BoneSpace
    translationMode=BMM_Additive
    translationSpace=BCS_BoneSpace
    alphaBoolBlend {blendInTime=0.5, blendOutTime=0.5}
    alphaScaleBiasClamp {bInterpResult=true, interpSpeedDecreasing=20, interpSpeedIncreasing=20}
    boneToModify {boneName=dyn_string_1_l}
    rotation=<rot 12.023 0.004 0.016>
   }
   n14 = "Transform (Modify) Bone - Bone: base_hoodie"(ComponentPose=@n13, Alpha=1, Translation="8.474,0.000,-1.871", Rotation="-14.442,0.000,-0.000", Scale="1,1,1") {
    rotationMode=BMM_Additive
    rotationSpace=BCS_BoneSpace
    scaleMode=BMM_Additive
    scaleSpace=BCS_BoneSpace
    translationMode=BMM_Additive
    translationSpace=BCS_BoneSpace
    alphaBoolBlend {blendInTime=0.5, blendOutTime=0.5}
    alphaScaleBiasClamp {bInterpResult=true}
    boneToModify {boneName=base_hoodie}
    rotation=<rot -14.442 0 0>
    translation {x=8.474, z=-1.871}
   }
   n15 = "Transform (Modify) Bone - Bone: base_hoodie"(ComponentPose=@n14, Alpha=1, Translation="11.551,-7.225,0.000", Rotation="-0.000,45.647,-0.000", Scale="1,1,1") {
    rotationMode=BMM_Additive
    rotationSpace=BCS_BoneSpace
    scaleMode=BMM_Additive
    scaleSpace=BCS_BoneSpace
    translationMode=BMM_Additive
    translationSpace=BCS_BoneSpace
    alphaBoolBlend {blendInTime=0.5, blendOutTime=0.5}
    alphaScaleBiasClamp {bInterpResult=true}
    boneToModify {boneName=base_hoodie}
    rotation=<rot 0 45.647 0>
    translation=<11.551 -7.225>
   }
   n16 = "Transform (Modify) Bone - Bone: base_hoodie"!(ComponentPose=@n15, bAlphaBoolEnabled:bool=false, Translation="-8,0,0", Rotation="50,0,0", Scale="1,1,1") -> (Pose) {
    alphaInputType=Bool
    rotationMode=BMM_Additive
    rotationSpace=BCS_BoneSpace
    scaleMode=BMM_Additive
    scaleSpace=BCS_BoneSpace
    translationMode=BMM_Additive
    translationSpace=BCS_BoneSpace
    alphaBoolBlend {blendInTime=0.5, blendOutTime=0.5}
    alphaScaleBiasClamp {bInterpResult=true}
    boneToModify {boneName=base_hoodie}
    rotation {pitch=50}
    translation {x=-8}
   }
   Output Pose(Result=Component To Local(ComponentPose=@n16)) {layerGroup=None}
  ]}
 EventGraph {nodes [
   Event BlueprintUpdateAnimation()
   > n1 = Sequence!(execute) -> (then_0, then_1, then_2:Exec, then_3:Exec)
   > n2 = "Gravity Override State Function (Parameter Asset) (Nativized)"(
    CharacterPartAnimInstance=Self-Reference()
    GameTimeInSeconds=GetGameTimeInSeconds()
    Parameters=&/F_MED_GoatFish_Parameters
   )
   > Set Gravity Override(Gravity Override=@n2)
   > Set RBAN Settings Out(RBAN Settings Out=@n2.OutRigidBodyAnimNodeParameters)
   Set Crouch Walk(execute=@n1.then_2, Crouch Walk:bool=AND Boolean(A=Get bIsCrouching(), B=Get bIsMoving2D())) -> (Output_Get:bool)
   n3 = Get bIsPlayingEmote()
   Set FEIdle(execute=@n1.then_3, FEIdle:bool=AND Boolean(A=Get bIsInFrontEnd(), B=NOT Boolean(A=@n3))) -> (Output_Get:bool)
   n4 = Get bIsCrouching()
   n5 = Get bIsTargeting()
   n6 = Get bIsMoving2D()
   > Set TargetingHeadUP(TargetingHeadUP:double=SelectFloat(A=Get Head Up() -> (Head Up:double), B=0, bPickA=AND Boolean!(A=@n5, B=@n4, C:bool=@n6) -> (ReturnValue))) -> (Output_Get:double)
   n7 = Get Skeletal_Mesh() -> (Skeletal_Mesh:Skeletal Mesh Component Object Reference)
   n8 = Set Head Up(
    execute=@n1.then_1
    Head Up:double=GetAlphaFromReferenceTransformDelta(SkeletalMeshComponent=@n7, BoneName=spine_05, BaseName=head, InputRange=(X=15,Y=70), TwistAxis="0,0,1")
   ) -> (Output_Get:double)
   > n9 = Set Head Down(Head Down:double=GetAlphaFromReferenceTransformDelta(SkeletalMeshComponent=@n7, BoneName=spine_05, BaseName="head ", InputRange=(X=0,Y=-90), OutputRange=(X=0,Y=-1), TwistAxis="0,0,1")) -> (Output_Get:double)
   > Set Head_Turn_L(Head_Turn_L:double=GetAlphaFromReferenceTransformDelta(SkeletalMeshComponent=@n7, BoneName=spine_05, BaseName=head, DeltaComponentType=EulerX, InputRange=(X=0,Y=-22), TwistAxis="1,0,0")) -> (Output_Get:double)
   > Set Head_Turn_R(Head_Turn_R:double=GetAlphaFromReferenceTransformDelta(SkeletalMeshComponent=@n7, BoneName=spine_05, BaseName=head, DeltaComponentType=EulerX, InputRange=(X=0,Y=22), TwistAxis="1,0,0")) -> (Output_Get:double)
   n10 = float + float(A=@n8, B=@n9)
   > Set Head look up alpha(Head look up alpha:double=SelectFloat(A="Clamp (Float)"(Value=@n10, Min=0, Max=1), B=0, bPickA=Get bIsTargeting())) -> (Output_Get:double)
   Event BlueprintInitializeAnimation()
   > Set Skeletal_Mesh(Skeletal_Mesh:Skeletal Mesh Component Object Reference=GetOwnerSkeletalMeshComponent()) -> (Output_Get:Skeletal Mesh Component Object Reference)
  ]}
]
variables [
 Crouch Walk {type=boolean}
 FEIdle {type=boolean}
 Gravity Override {type=Vector}
 Head Down {type=number}
 Head look up alpha {type=number}
 Head Up {type=number}
 Head_Turn_L {type=number}
 Head_Turn_R {type=number}
 RBAN Settings Out {type=FortRigidBodyAnimNodeParameters}
 Skeletal_Mesh {type=SkeletalMeshComponent}
 TargetingHeadUP {type=number}
]
