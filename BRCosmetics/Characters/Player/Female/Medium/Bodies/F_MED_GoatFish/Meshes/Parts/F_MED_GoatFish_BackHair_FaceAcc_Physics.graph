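# Physics asset for the back hair of the female medium GoatFish character's face accessory, giving it three hair chains for secondary motion.
# Left, middle and right chains of three dyn bones hang from base bones linked to the head, with narrow swing on one axis, wide on the other and zero twist.
# Cross constraints tie each side segment to its middle counterpart so the strands move together.
# Capsules on head, neck, clavicles and upper spine approximate the body.
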PhysicsAsset /BRCosmetics/Characters/Player/Female/Medium/Bodies/F_MED_GoatFish/Meshes/Parts/F_MED_GoatFish_BackHair_FaceAcc_Physics
bodies {
 base_hair_back_l [{shapeType=Capsule, length=4, radius=3, shapeName=base_hair_back_l_capsule, center=<3 0 0>, rotation=<rot 90 0 0>}]
 base_hair_back_r [{shapeType=Capsule, length=7, radius=3, shapeName=base_hair_back_r_capsule, center=<3 0 0>, rotation=<rot 90 0 90>}]
 base_hair_mid [{shapeType=Capsule, length=7, radius=3, shapeName=base_hair_mid_capsule, center=<3 0 0>, rotation=<rot 90 0 90>}]
 clavicle_l [
  {shapeType=Capsule, length=8.9288, radius=6.2603, shapeName=clavicle_l_capsule, center=<5.195 -0.2576 -1.146>, rotation=<rot 90 -11.2795 0>}
  {
   shapeType=Capsule
   length=6.9008
   radius=3.9413
   shapeName=clavicle_l_capsule
   center=<5.5946 -5.3424 3.8899>
   rotation=<rot -77.4179 66.5442 -153.4774>
  }
 ]
 clavicle_r [{shapeType=Capsule, length=9.5817, radius=5.4183, shapeName=clavicle_r_capsule, center=<-5 1.1945 0.3817>, rotation=<rot 90 0 0>}]
 dyn_hair_back_1_l [{shapeType=Capsule, length=2.4347, radius=4, shapeName=dyn_hair_back_1_l_capsule, center=<3.3758 0 1>, rotation=<rot 90 0 0>}]
 dyn_hair_back_1_r [{shapeType=Capsule, length=3, radius=4.5, shapeName=dyn_hair_back_1_r_capsule, center=<3 0 0>, rotation=<rot 90 0 0>}]
 dyn_hair_back_2_l [{shapeType=Capsule, length=2.0075, radius=4, shapeName=dyn_hair_back_2_l_capsule, center=<1 0 1.5>, rotation=<rot 90 0 90>}]
 dyn_hair_back_2_r [{shapeType=Capsule, length=6.8335, radius=4.5, shapeName=dyn_hair_back_2_r_capsule, center=<1 0 1>, rotation=<rot 90 0 90>}]
 dyn_hair_back_3_l [
  {
   shapeType=Capsule
   length=6.2072
   radius=4.2928
   shapeName=dyn_hair_back_3_l_capsule
   center=<3 0 2>
   rotation=<rot 59.6091 -46.5161 -137.6627>
  }
  {
   shapeType=Capsule
   length=3.3895
   radius=3.9313
   shapeName=dyn_hair_back_3_l_capsule
   center=<0.7772 3.7539 1.8125>
   rotation=<rot 52.5609 4.4251 -17.1369>
  }
 ]
 dyn_hair_back_3_r [{shapeType=Capsule, length=6.9238, radius=4.4438, shapeName=dyn_hair_back_3_r_capsule, center=<2.5 0 1>, rotation=<rot 70 0 95>}]
 dyn_hair_mid_1 [{shapeType=Capsule, length=4, radius=4.3, shapeName=dyn_hair_mid_1_capsule, center=<3.5 0 1>, rotation=<rot 90 0 0>}]
 dyn_hair_mid_2 [{shapeType=Capsule, length=6, radius=5, shapeName=dyn_hair_mid_2_capsule, center=<4 0 1>, rotation=<rot 90 0 90>}]
 dyn_hair_mid_3 [
  {shapeType=Capsule, length=7.8427, radius=4.7438, shapeName=dyn_hair_mid_3_capsule, center=<3 0 0>, rotation=<rot 85 0 94>}
  {
   shapeType=Capsule
   length=12.0198
   radius=2.9802
   shapeName=dyn_hair_mid_3_capsule
   center=<-2.1177 -0.0061 0.1333>
   rotation=<rot -0.0001 0 92.9339>
  }
 ]
 head [{shapeType=Capsule, length=4.629, radius=4, shapeName=head_capsule, rotation=<rot 90 15.3549 0>}]
 neck_02 [{shapeType=Capsule, length=2.671, radius=4.5, shapeName=neck_02_capsule, rotation=<rot 90 0 0>}]
 spine_05 [
  {shapeType=Capsule, length=11.7219, radius=11.2781, shapeName=spine_05_capsule, center=<-0.8625 -2.1721 0>, rotation=<rot 0 -90 0>}
  {shapeType=Capsule, length=15.6845, radius=10.0815, shapeName=spine_05_capsule, center=<3.4868 -7.6611 0>}
  {shapeType=Capsule, length=5.5, radius=3, shapeName=spine_05_capsule, center=<8.4 2.2 -5.7>, rotation=<rot 43 -73 -8.2>}
  {shapeType=Capsule, length=7.5, radius=3.8, shapeName=spine_05_capsule, center=<9.7 -5 -9>, rotation=<rot -0.1 -3.5 80>}
  {
   shapeType=Capsule
   length=8.9509
   radius=4.3907
   shapeName=spine_05_capsule
   center=<9.2439 -5 9.0016>
   rotation=<rot 0.0414 -4.3018 -79.9996>
  }
  {shapeType=Capsule, length=5.5, radius=3, shapeName=spine_05_capsule, center=<8.4 2.2 5.7>, rotation=<rot -43 -73 -8.2>}
 ]
}
constraints [
 {bone1Name=dyn_hair_mid_3, bone2Name=dyn_hair_mid_2, swing1LimitDegrees=10, swing2LimitDegrees=50, twistLimitDegrees=0}
 {bone1Name=dyn_hair_mid_1, bone2Name=base_hair_mid, swing1LimitDegrees=10, swing2LimitDegrees=80, twistLimitDegrees=0}
 {bone1Name=dyn_hair_mid_2, bone2Name=dyn_hair_mid_1, swing1LimitDegrees=10, swing2LimitDegrees=80, twistLimitDegrees=0}
 {bone1Name=dyn_hair_back_3_r, bone2Name=dyn_hair_back_2_r, swing1LimitDegrees=10, swing2LimitDegrees=50, twistLimitDegrees=0}
 {bone1Name=dyn_hair_back_1_r, bone2Name=base_hair_back_r, swing1LimitDegrees=10, swing2LimitDegrees=80, twistLimitDegrees=0}
 {bone1Name=dyn_hair_back_2_r, bone2Name=dyn_hair_back_1_r, swing1LimitDegrees=10, swing2LimitDegrees=80, twistLimitDegrees=0}
 {bone1Name=dyn_hair_back_3_l, bone2Name=dyn_hair_back_2_l, swing1LimitDegrees=10, swing2LimitDegrees=50, twistLimitDegrees=0}
 {bone1Name=dyn_hair_back_1_l, bone2Name=base_hair_back_l, swing1LimitDegrees=10, swing2LimitDegrees=90, twistLimitDegrees=0}
 {bone1Name=dyn_hair_back_2_l, bone2Name=dyn_hair_back_1_l, swing1LimitDegrees=10, swing2LimitDegrees=80, twistLimitDegrees=0}
 {
  bone1Name=base_hair_mid
  bone2Name=head
  swing1LimitDegrees=45
  swing1Motion=Free
  swing2LimitDegrees=45
  swing2Motion=Free
  twistLimitDegrees=45
  twistMotion=Free
 }
 {
  bone1Name=base_hair_back_l
  bone2Name=head
  swing1LimitDegrees=45
  swing1Motion=Free
  swing2LimitDegrees=45
  swing2Motion=Free
  twistLimitDegrees=45
  twistMotion=Free
 }
 {
  bone1Name=base_hair_back_r
  bone2Name=head
  swing1LimitDegrees=45
  swing1Motion=Free
  swing2LimitDegrees=45
  swing2Motion=Free
  twistLimitDegrees=45
  twistMotion=Free
 }
 {
  bone1Name=head
  bone2Name=neck_02
  swing1LimitDegrees=45
  swing1Motion=Free
  swing2LimitDegrees=45
  swing2Motion=Free
  twistLimitDegrees=45
  twistMotion=Free
 }
 {
  bone1Name=neck_02
  bone2Name=spine_05
  swing1LimitDegrees=45
  swing1Motion=Free
  swing2LimitDegrees=45
  swing2Motion=Free
  twistLimitDegrees=45
  twistMotion=Free
 }
 {
  bone1Name=clavicle_r
  bone2Name=spine_05
  swing1LimitDegrees=45
  swing1Motion=Free
  swing2LimitDegrees=45
  swing2Motion=Free
  twistLimitDegrees=45
  twistMotion=Free
 }
 {
  bone1Name=clavicle_l
  bone2Name=spine_05
  swing1LimitDegrees=45
  swing1Motion=Free
  swing2LimitDegrees=45
  swing2Motion=Free
  twistLimitDegrees=45
  twistMotion=Free
 }
 {bone1Name=dyn_hair_back_1_r, bone2Name=dyn_hair_mid_1, swing1LimitDegrees=45, swing2LimitDegrees=45, twistLimitDegrees=45}
 {bone1Name=dyn_hair_back_1_l, bone2Name=dyn_hair_mid_1, swing1LimitDegrees=45, swing2LimitDegrees=45, twistLimitDegrees=45}
 {bone1Name=dyn_hair_back_2_r, bone2Name=dyn_hair_mid_2, swing1LimitDegrees=45, swing2LimitDegrees=45, twistLimitDegrees=45}
 {bone1Name=dyn_hair_back_2_l, bone2Name=dyn_hair_mid_2, swing1LimitDegrees=45, swing2LimitDegrees=45, twistLimitDegrees=45}
 {bone1Name=dyn_hair_back_3_r, bone2Name=dyn_hair_mid_3, swing1LimitDegrees=45, swing2LimitDegrees=45, twistLimitDegrees=45}
 {bone1Name=dyn_hair_back_3_l, bone2Name=dyn_hair_mid_3, swing1LimitDegrees=45, swing2LimitDegrees=45, twistLimitDegrees=45}
]
