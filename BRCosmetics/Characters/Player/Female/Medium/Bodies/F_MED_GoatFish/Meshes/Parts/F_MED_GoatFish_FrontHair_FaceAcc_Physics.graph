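# Physics asset for the front hair and face accessory part of the GoatFish female medium character, defining physics bodies for the left and right side hair.
# Each side chains a base and three capsule segments with limited swing and zero twist, generally looser on the left.
# The right base is locked to the head, while the left is free.
# Capsules on head, neck, upper spine, clavicles, upper arms and right forearm, joined by free constraints, approximate the body.
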PhysicsAsset /BRCosmetics/Characters/Player/Female/Medium/Bodies/F_MED_GoatFish/Meshes/Parts/F_MED_GoatFish_FrontHair_FaceAcc_Physics
bodies {
 base_hair_side_l [{shapeType=Capsule, length=7, radius=3, shapeName=base_hair_side_l_capsule, center=<3 0 0>, rotation=<rot 90 0 90>}]
 base_hair_side_r [{shapeType=Capsule, length=7, radius=3, shapeName=base_hair_side_r_capsule, center=<3 0 0>, rotation=<rot 90 0 90>}]
 clavicle_l [{shapeType=Capsule, length=8.9288, radius=6.2603, shapeName=clavicle_l_capsule, center=<5.195 -0.2576 -1.146>, rotation=<rot 90 -11.2795 0>}]
 clavicle_r [
  {shapeType=Capsule, length=9.5817, radius=5.4183, shapeName=clavicle_r_capsule, center=<-5 1.1945 0.3817>, rotation=<rot 90 0 0>}
  {
   shapeType=Capsule
   length=8.1313
   radius=3.3586
   shapeName=clavicle_r_capsule
   center=<-2.6844 -1.8661 -3.4378>
   rotation=<rot -26.415 39.6118 -59.8273>
  }
 ]
 dyn_hair_side_1_l [{shapeType=Capsule, length=4, radius=3, shapeName=dyn_hair_side_1_l_capsule, center=<2 0 0>, rotation=<rot 90 0 0>}]
 dyn_hair_side_2_l [{
   shapeType=Capsule
   length=4.3277
   radius=2.6723
   shapeName=dyn_hair_side_2_l_capsule
   center=<1.3 -1.7 0>
   rotation=<rot 79.3063 160.376 -180>
  }]
 dyn_hair_side_3_l [
  {shapeType=Capsule, length=8.8558, radius=2.3657, shapeName=dyn_hair_side_3_l_capsule, center=<2 0 0>, rotation=<rot 46.7789 49.8911 0>}
  {shapeType=Capsule, length=3.6794, radius=1.9206, shapeName=dyn_hair_side_3_l_capsule, center=<-1.2 -6 4.745>, rotation=<rot 90 -33.455 0>}
 ]
 hair_side_1_r [{shapeType=Capsule, length=2, radius=3, shapeName=hair_side_1_r_capsule, center=<2 0 0>, rotation=<rot -90 0 0>}]
 hair_side_2_r [{shapeType=Capsule, length=2.417, radius=2.6617, shapeName=hair_side_2_r_capsule, center=<2.9017 0 0>, rotation=<rot 90 0 0>}]
 hair_side_3_r [
  {
   shapeType=Capsule
   length=4.0336
   radius=1.9979
   shapeName=hair_side_3_r_capsule
   center=<0.5737 -3.0452 -2.4116>
   rotation=<rot 50.4892 61.8879 98.3923>
  }
  {
   shapeType=Capsule
   length=4.3138
   radius=2.0442
   shapeName=hair_side_3_r_capsule
   center=<-0.0362 -0.5183 2.6634>
   rotation=<rot -33.3676 -60.0778 69.4645>
  }
 ]
 head [
  {shapeType=Capsule, length=4.629, radius=4, shapeName=head_capsule, rotation=<rot 90 15.3549 0>}
  {shapeType=Capsule, length=4.5368, radius=6.0622, shapeName=head_capsule, center=<-0.7369 6.5679 1.606>, rotation=<rot 0 -90 90.1399>}
 ]
 lowerarm_r [{shapeType=Capsule, length=25, radius=7, shapeName=lowerarm_r_capsule, center=<-10 0 0>, rotation=<rot 90 0 0>}]
 neck_02 [{shapeType=Capsule, length=5.4214, radius=4.5, shapeName=neck_02_capsule, center=<-1.7796 0 0>, rotation=<rot 90 0 0>}]
 spine_05 [
  {shapeType=Capsule, length=11.7219, radius=11.2781, shapeName=spine_05_capsule, center=<-0.8625 -2.1721 0>, rotation=<rot 0 -90 0>}
  {shapeType=Capsule, length=15.8855, radius=9.8806, shapeName=spine_05_capsule, center=<3.2501 -7.6611 0>}
  {shapeType=Capsule, length=5.5, radius=3, shapeName=spine_05_capsule, center=<8.4 2.2 -5.7>, rotation=<rot 43 -73 -8.2>}
  {shapeType=Capsule, length=7.5, radius=3.8, shapeName=spine_05_capsule, center=<9 -5 -9>, rotation=<rot -0.1 -3.5 80>}
  {
   shapeType=Capsule
   length=8.9509
   radius=4.3907
   shapeName=spine_05_capsule
   center=<8.4713 -4.9681 9.001>
   rotation=<rot 0.0414 -4.3018 -79.9996>
  }
  {shapeType=Capsule, length=5.5, radius=3, shapeName=spine_05_capsule, center=<8.4 2.2 5.7>, rotation=<rot -43 -73 16.7996>}
 ]
 upperarm_l [{shapeType=Capsule, length=16.739, radius=5, shapeName=upperarm_l_capsule, center=<10.0642 0 0>, rotation=<rot -90 0 0>}]
 upperarm_r [{
   shapeType=Capsule
   length=16.7555
   radius=5.8438
   shapeName=upperarm_r_capsule
   center=<-10.3504 0.4075 0.0012>
   rotation=<rot -85.214 0.1574 -2.0772>
  }]
}
constraints [
 {bone1Name=hair_side_3_r, bone2Name=hair_side_2_r, swing1LimitDegrees=10, swing2LimitDegrees=10, twistLimitDegrees=0}
 {bone1Name=hair_side_2_r, bone2Name=hair_side_1_r, swing1LimitDegrees=10, swing2LimitDegrees=10, twistLimitDegrees=0}
 {bone1Name=dyn_hair_side_3_l, bone2Name=dyn_hair_side_2_l, swing1LimitDegrees=20, swing2LimitDegrees=20, twistLimitDegrees=0}
 {bone1Name=dyn_hair_side_1_l, bone2Name=base_hair_side_l, swing1LimitDegrees=5, swing2LimitDegrees=5, twistLimitDegrees=0}
 {bone1Name=dyn_hair_side_2_l, bone2Name=dyn_hair_side_1_l, swing1LimitDegrees=20, swing2LimitDegrees=20, twistLimitDegrees=0}
 {
  bone1Name=base_hair_side_l
  bone2Name=head
  swing1LimitDegrees=45
  swing1Motion=Free
  swing2LimitDegrees=45
  swing2Motion=Free
  twistLimitDegrees=45
  twistMotion=Free
 }
 {
  bone1Name=base_hair_side_r
  bone2Name=head
  swing1LimitDegrees=45
  swing1Motion=Locked
  swing2LimitDegrees=45
  swing2Motion=Locked
  twistLimitDegrees=45
  twistMotion=Locked
 }
 {
  bone1Name=head
  bone2Name=neck_02
  swing1LimitDegrees=45
  swing1Motion=Free
  swing2LimitDegrees=45
  swing2Motion=Free
  twistLimitDegrees=45
  twistMotion=Free
 }
 {
  bone1Name=neck_02
  bone2Name=spine_05
  swing1LimitDegrees=45
  swing1Motion=Free
  swing2LimitDegrees=45
  swing2Motion=Free
  twistLimitDegrees=45
  twistMotion=Free
 }
 {
  bone1Name=clavicle_r
  bone2Name=spine_05
  swing1LimitDegrees=45
  swing1Motion=Free
  swing2LimitDegrees=45
  swing2Motion=Free
  twistLimitDegrees=45
  twistMotion=Free
 }
 {
  bone1Name=upperarm_r
  bone2Name=clavicle_r
  swing1LimitDegrees=45
  swing1Motion=Free
  swing2LimitDegrees=45
  swing2Motion=Free
  twistLimitDegrees=45
  twistMotion=Free
 }
 {
  bone1Name=clavicle_l
  bone2Name=spine_05
  swing1LimitDegrees=45
  swing1Motion=Free
  swing2LimitDegrees=45
  swing2Motion=Free
  twistLimitDegrees=45
  twistMotion=Free
 }
 {
  bone1Name=upperarm_l
  bone2Name=clavicle_l
  swing1LimitDegrees=45
  swing1Motion=Free
  swing2LimitDegrees=45
  swing2Motion=Free
  twistLimitDegrees=45
  twistMotion=Free
 }
 {bone1Name=hair_side_1_r, bone2Name=base_hair_side_r, swing1LimitDegrees=2, swing2LimitDegrees=2, twistLimitDegrees=0}
 {
  bone1Name=lowerarm_r
  bone2Name=upperarm_r
  swing1LimitDegrees=45
  swing1Motion=Free
  swing2LimitDegrees=45
  swing2Motion=Free
  twistLimitDegrees=45
  twistMotion=Free
 }
]
