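# Post-process-style AnimBP for the GoatFish female face accessory, animating hair and charms on top of the copied parent pose.
# A Control Rig feeds front/back hair RigidBody sims, per-bone AnimDynamics, corrective bone offsets and weighted head/spine/clavicle constraints, then a post Control Rig.
# The EventGraph derives head pitch, crouch, aiming, pistol and backbling state for the rigs.
# Crouching zeroes the back-hair RigidBody alpha.
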AnimBlueprint /BRCosmetics/Characters/Player/Female/Medium/Bodies/F_MED_GoatFish/Meshes/Parts/F_MED_GoatFish_FaceAcc_AnimBP
target_skeleton=&/F_MED_GoatFish_FaceAcc_Skeleton
graphs [
 AnimGraph {nodes [
   n1 = Control Rig!(
    Source=Copy Pose From Mesh(SourceMeshComponent=Get Skeletal_Mesh() -> (Skeletal_Mesh:Skeletal Mesh Component Object Reference)) {bUseAttachedParent=true}
    Alpha
    CrouchheadUP:double=Get CrouchHeadUP() -> (CrouchHeadUP:double)
    SkyDiving:bool=Get bIsSkydiveFloating()
    SkyDive_Parachuting:bool=Get bParachuteOpenedAndStandingUpright() -> (bParachuteOpenedAndStandingUpright:bool)
    TargetingHeadUP:double=Get TargetingHeadUP() -> (TargetingHeadUP:double)
    Crouching:bool=Get bIsCrouching()
    SkyFloating:bool=Get bIsSkydiveFloating()
    CrouchNotWalk:bool=Get CrouchNOt walk() -> (CrouchNOt walk:bool)
    Crouchwalk:bool=Get Crouchwalk() -> (Crouchwalk:bool)
    CrouchWalkTarget:bool=false
    PistolHeadUP:double=Get PistolHeadUP() -> (PistolHeadUP:double)
    UsingPistol:double=Get Pistol Equipped_Float() -> (Pistol Equipped_Float:double)
    IsTargeting:double=Get Istargeting_Float() -> (Istargeting_Float:double)
   ) -> (Pose) {controlRigAssetReference {blueprintRigClass=&/F_MED_GoatFish_FaceAcc_CtrlRig.F_MED_GoatFish_FaceAcc_CtrlRig_C}}
   n2 = RigidBody!(ComponentPose=Local To Component(LocalPose=@n1), Alpha, OverrideWorldGravity:Vector="0,0,0", ExternalForce) -> (Pose) {
    bOverrideWorldGravity=true
    overridePhysicsAsset=&/F_MED_GoatFish_FrontHair_FaceAcc_Physics
    simulationSpace=BaseBoneSpace
    baseBoneRef {boneName=spine_05}
    simSpaceSettings {velocityScaleZ=0.8, worldAlpha=0.8}
   }
   n3 = RigidBody!(
    ComponentPose=@n2
    Alpha=Get RigidOFF() -> (RigidOFF:double)
    OverrideWorldGravity="0,0,0"
    ExternalForce
    SimSpaceSettings:Sim Space Settings Structure=(WorldAlpha=1,VelocityScaleZ=0.8,DampingAlpha=1,MaxLinearVelocity=10000,MaxAngularVelocity=10000,MaxLinearAcceleration=10000,MaxAngularAcceleration=10000,ExternalLinearDragV=(X=0,Y=2,Z=0),ExternalLinearVelocity=(X=0,Y=0,Z=0),ExternalAngularVelocity=(X=0,Y=0,Z=0))
   ) -> (Pose) {
    bOverrideWorldGravity=true
    overridePhysicsAsset=&/F_MED_GoatFish_BackHair_FaceAcc_Physics
    simulationSpace=BaseBoneSpace
    baseBoneRef {boneName=spine_05}
    simSpaceSettings {velocityScaleZ=0.8, worldAlpha=1, externalLinearDragV {y=2}}
   }
   n4 = Anim Dynamics - Bone: dyn_charm_1!(ComponentPose=@n3, Alpha=1, GravityOverride:Vector="0,0,-980", ExternalForce="0,0,0") -> (Pose) {
    angularDampingOverride=1
    angularSpringConstant=12000
    bAngularSpring=true
    bOverrideAngularDamping=true
    bUseGravityOverride=true
    gravityScale=0.35
    simulationSpace=RootRelative
    boundBone {boneName=dyn_charm_1}
    componentAppliedLinearAccClamp=<3000 3000 3000>
    componentLinearAccScale=<3 3 3>
    componentLinearVelScale=<3 3 3>
    gravityOverride {z=-980}
    physicsBodyDefinitions [{
      boundBone {boneName=dyn_charm_1}
      boxExtents=<20 20 20>
      constraintSetup {angularLimitsMax=<15 100 100>, angularLimitsMin=<-15 -100 -100>, angularTarget=<1 0 0>}
      localJointOffset=<5 0 0>
     }]
   }
   n5 = Anim Dynamics - Bone: dyn_charm_2!(ComponentPose=@n4, Alpha=1, GravityOverride:Vector="0,0,-980", ExternalForce="0,0,0") -> (Pose) {
    angularDampingOverride=1
    angularSpringConstant=12000
    bAngularSpring=true
    bOverrideAngularDamping=true
    bUseGravityOverride=true
    gravityScale=0.35
    simulationSpace=RootRelative
    boundBone {boneName=dyn_charm_2}
    componentAppliedLinearAccClamp=<3000 3000 3000>
    componentLinearAccScale=<3 3 3>
    componentLinearVelScale=<3 3 3>
    gravityOverride {z=-980}
    physicsBodyDefinitions [{
      boundBone {boneName=dyn_charm_2}
      boxExtents=<20 20 20>
      constraintSetup {angularLimitsMax=<15 80 100>, angularLimitsMin=<-15 -100 -100>, angularTarget=<1 0 0>}
      localJointOffset=<5 0 0>
     }]
   }
   n6 = Anim Dynamics - Bone: dyn_hair_Forehead_1!(ComponentPose=@n5, Alpha=1, GravityOverride:Vector="0,0,-980", ExternalForce="0,0,0") -> (Pose) {
    angularDampingOverride=1
    angularSpringConstant=15000
    bAngularSpring=true
    bOverrideAngularDamping=true
    bUseGravityOverride=true
    gravityScale=0.35
    simulationSpace=RootRelative
    boundBone {boneName=dyn_hair_Forehead_1}
    componentAppliedLinearAccClamp=<3000 3000 3000>
    componentLinearAccScale=<0.2 0.2 0.2>
    componentLinearVelScale=<0.2 0.2 0.2>
    gravityOverride {z=-980}
    physicsBodyDefinitions [{
      boundBone {boneName=dyn_hair_Forehead_1}
      boxExtents=<20 30 30>
      constraintSetup {angularLimitsMax=<0 0 5>, angularLimitsMin=<0 -5 -5>, angularTarget=<1 0 0>}
      localJointOffset=<4 0 0>
     }]
   }
   n7 = Anim Dynamics - Bone: dyn_hair_Forehead_2!(ComponentPose=@n6, Alpha=1, GravityOverride:Vector="0,0,-980", ExternalForce="0,0,0") -> (Pose) {
    angularDampingOverride=1
    angularSpringConstant=15000
    bAngularSpring=true
    bOverrideAngularDamping=true
    bUseGravityOverride=true
    gravityScale=0.35
    simulationSpace=RootRelative
    boundBone {boneName=dyn_hair_Forehead_2}
    componentAppliedLinearAccClamp=<3000 3000 3000>
    componentLinearAccScale=<0.2 0.2 0.2>
    componentLinearVelScale=<0.2 0.2 0.2>
    gravityOverride {z=-980}
    physicsBodyDefinitions [{
      boundBone {boneName=dyn_hair_Forehead_2}
      boxExtents=<20 30 30>
      constraintSetup {angularLimitsMax=<0 0 5>, angularLimitsMin=<0 -5 -5>, angularTarget=<1 0 0>}
      localJointOffset=<4 0 0>
     }]
   }
   n8 = Anim Dynamics - Bone: dyn_hair_back_1_r!(ComponentPose=@n7, bAlphaBoolEnabled:bool=True, GravityOverride:Vector="0,0,-980", ExternalForce="0,0,0") -> (Pose) {
    alphaInputType=Bool
    angularDampingOverride=1
    angularSpringConstant=10000
    bAngularSpring=true
    bOverrideAngularDamping=true
    bUseGravityOverride=true
    gravityScale=0.35
    simulationSpace=RootRelative
    boundBone {boneName=dyn_hair_back_1_r}
    componentAppliedLinearAccClamp=<3000 3000 3000>
    componentLinearAccScale=<1 1 1>
    componentLinearVelScale=<1 1 1>
    gravityOverride {z=-980}
    physicsBodyDefinitions [{
      boundBone {boneName=dyn_hair_back_1_r}
      boxExtents=<20 30 30>
      constraintSetup {angularLimitsMax=<0 0 5>, angularLimitsMin=<0 -15 -5>, angularTarget=<1 0 0>}
      localJointOffset=<4 0 0>
     }]
   }
   n9 = Anim Dynamics - Bone: dyn_hair_back_2_r!(ComponentPose=@n8, bAlphaBoolEnabled:bool=True, GravityOverride:Vector="0,0,-980", ExternalForce="0,0,0") -> (Pose) {
    alphaInputType=Bool
    angularDampingOverride=1
    angularSpringConstant=10000
    bAngularSpring=true
    bOverrideAngularDamping=true
    bUseGravityOverride=true
    gravityScale=0.35
    simulationSpace=RootRelative
    boundBone {boneName=dyn_hair_back_2_r}
    componentAppliedLinearAccClamp=<3000 3000 3000>
    componentLinearAccScale=<1 1 1>
    componentLinearVelScale=<1 1 1>
    gravityOverride {z=-980}
    physicsBodyDefinitions [{
      boundBone {boneName=dyn_hair_back_2_r}
      boxExtents=<20 30 30>
      constraintSetup {angularLimitsMax=<0 0 5>, angularLimitsMin=<0 -15 -5>, angularTarget=<1 0 0>}
      localJointOffset=<4 0 0>
     }]
   }
   n10 = Anim Dynamics - Bone: dyn_hair_back_3_r!(ComponentPose=@n9, bAlphaBoolEnabled:bool=True, GravityOverride:Vector="0,0,-980", ExternalForce="0,0,0") -> (Pose) {
    alphaInputType=Bool
    angularDampingOverride=1
    angularSpringConstant=10000
    bAngularSpring=true
    bOverrideAngularDamping=true
    bUseGravityOverride=true
    gravityScale=0.35
    simulationSpace=RootRelative
    boundBone {boneName=dyn_hair_back_3_r}
    componentAppliedLinearAccClamp=<3000 3000 3000>
    componentLinearAccScale=<1 1 1>
    componentLinearVelScale=<1 1 1>
    gravityOverride {z=-980}
    physicsBodyDefinitions [{
      boundBone {boneName=dyn_hair_back_3_r}
      boxExtents=<20 30 30>
      constraintSetup {angularLimitsMax=<0 0 5>, angularLimitsMin=<0 -15 -5>, angularTarget=<1 0 0>}
      localJointOffset=<4 0 0>
     }]
   }
   n11 = Anim Dynamics - Bone: dyn_hair_back_4_r!(ComponentPose=@n10, bAlphaBoolEnabled:bool=True, GravityOverride:Vector="0,0,-980", ExternalForce="0,0,0") -> (Pose) {
    alphaInputType=Bool
    angularDampingOverride=1
    angularSpringConstant=10000
    bAngularSpring=true
    bOverrideAngularDamping=true
    bUseGravityOverride=true
    gravityScale=0.35
    simulationSpace=RootRelative
    boundBone {boneName=dyn_hair_back_4_r}
    componentAppliedLinearAccClamp=<3000 3000 3000>
    componentLinearAccScale=<1 1 1>
    componentLinearVelScale=<1 1 1>
    gravityOverride {z=-980}
    physicsBodyDefinitions [{
      boundBone {boneName=dyn_hair_back_4_r}
      boxExtents=<20 30 30>
      constraintSetup {angularLimitsMax=<0 0 5>, angularLimitsMin=<0 -15 -5>, angularTarget=<1 0 0>}
      localJointOffset=<4 0 0>
     }]
   }
   n12 = Anim Dynamics - Bone: dyn_hair_back_1_l!(ComponentPose=@n11, bAlphaBoolEnabled:bool=True, GravityOverride:Vector="0,0,-980", ExternalForce="0,0,0") -> (Pose) {
    alphaInputType=Bool
    angularDampingOverride=1
    angularSpringConstant=10000
    bAngularSpring=true
    bOverrideAngularDamping=true
    bUseGravityOverride=true
    gravityScale=0.35
    simulationSpace=RootRelative
    boundBone {boneName=dyn_hair_back_1_l}
    componentAppliedLinearAccClamp=<3000 3000 3000>
    componentLinearAccScale=<1 1 1>
    componentLinearVelScale=<1 1 1>
    gravityOverride {z=-980}
    physicsBodyDefinitions [{
      boundBone {boneName=dyn_hair_back_1_l}
      boxExtents=<20 30 30>
      constraintSetup {angularLimitsMax=<0 0 5>, angularLimitsMin=<0 -15 -5>, angularTarget=<1 0 0>}
      localJointOffset=<4 0 0>
     }]
   }
   n13 = Anim Dynamics - Bone: dyn_hair_back_2_l!(ComponentPose=@n12, bAlphaBoolEnabled:bool=True, GravityOverride:Vector="0,0,-980", ExternalForce="0,0,0") -> (Pose) {
    alphaInputType=Bool
    angularDampingOverride=1
    angularSpringConstant=10000
    bAngularSpring=true
    bOverrideAngularDamping=true
    bUseGravityOverride=true
    gravityScale=0.35
    simulationSpace=RootRelative
    boundBone {boneName=dyn_hair_back_2_l}
    componentAppliedLinearAccClamp=<3000 3000 3000>
    componentLinearAccScale=<1 1 1>
    componentLinearVelScale=<1 1 1>
    gravityOverride {z=-980}
    physicsBodyDefinitions [{
      boundBone {boneName=dyn_hair_back_2_l}
      boxExtents=<20 30 30>
      constraintSetup {angularLimitsMax=<0 0 5>, angularLimitsMin=<0 -15 -5>, angularTarget=<1 0 0>}
      localJointOffset=<4 0 0>
     }]
   }
   n14 = Anim Dynamics - Bone: dyn_hair_back_3_l!(ComponentPose=@n13, bAlphaBoolEnabled:bool=True, GravityOverride:Vector="0,0,-980", ExternalForce="0,0,0") -> (Pose) {
    alphaInputType=Bool
    angularDampingOverride=1
    angularSpringConstant=10000
    bAngularSpring=true
    bOverrideAngularDamping=true
    bUseGravityOverride=true
    gravityScale=0.35
    simulationSpace=RootRelative
    boundBone {boneName=dyn_hair_back_3_l}
    componentAppliedLinearAccClamp=<3000 3000 3000>
    componentLinearAccScale=<1 1 1>
    componentLinearVelScale=<1 1 1>
    gravityOverride {z=-980}
    physicsBodyDefinitions [{
      boundBone {boneName=dyn_hair_back_3_l}
      boxExtents=<20 30 30>
      constraintSetup {angularLimitsMax=<0 0 5>, angularLimitsMin=<0 -15 -5>, angularTarget=<1 0 0>}
      localJointOffset=<4 0 0>
     }]
   }
   n15 = Anim Dynamics - Bone: dyn_hair_back_4_l!(ComponentPose=@n14, bAlphaBoolEnabled:bool=True, GravityOverride:Vector="0,0,-980", ExternalForce="0,0,0") -> (Pose) {
    alphaInputType=Bool
    angularDampingOverride=1
    angularSpringConstant=10000
    bAngularSpring=true
    bOverrideAngularDamping=true
    bUseGravityOverride=true
    gravityScale=0.35
    simulationSpace=RootRelative
    boundBone {boneName=dyn_hair_back_4_l}
    componentAppliedLinearAccClamp=<3000 3000 3000>
    componentLinearAccScale=<1 1 1>
    componentLinearVelScale=<1 1 1>
    gravityOverride {z=-980}
    physicsBodyDefinitions [{
      boundBone {boneName=dyn_hair_back_4_l}
      boxExtents=<20 30 30>
      constraintSetup {angularLimitsMax=<0 0 5>, angularLimitsMin=<0 -15 -5>, angularTarget=<1 0 0>}
      localJointOffset=<4 0 0>
     }]
   }
   n16 = Anim Dynamics - Bone: dyn_hair_mid_1!(ComponentPose=@n15, bAlphaBoolEnabled:bool=True, GravityOverride:Vector="0,0,-980", ExternalForce="0,0,0") -> (Pose) {
    alphaInputType=Bool
    angularDampingOverride=1
    angularSpringConstant=10000
    bAngularSpring=true
    bOverrideAngularDamping=true
    bUseGravityOverride=true
    gravityScale=0.35
    simulationSpace=RootRelative
    boundBone {boneName=dyn_hair_mid_1}
    componentAppliedLinearAccClamp=<3000 3000 3000>
    componentLinearAccScale=<1 1 1>
    componentLinearVelScale=<1 1 1>
    gravityOverride {z=-980}
    physicsBodyDefinitions [{
      boundBone {boneName=dyn_hair_mid_1}
      boxExtents=<20 30 30>
      constraintSetup {angularLimitsMax=<0 0 5>, angularLimitsMin=<0 -15 -5>, angularTarget=<1 0 0>}
      localJointOffset=<4 0 0>
     }]
   }
   n17 = Anim Dynamics - Bone: dyn_hair_mid_2!(ComponentPose=@n16, bAlphaBoolEnabled:bool=True, GravityOverride:Vector="0,0,-980", ExternalForce="0,0,0") -> (Pose) {
    alphaInputType=Bool
    angularDampingOverride=1
    angularSpringConstant=10000
    bAngularSpring=true
    bOverrideAngularDamping=true
    bUseGravityOverride=true
    gravityScale=0.35
    simulationSpace=RootRelative
    boundBone {boneName=dyn_hair_mid_2}
    componentAppliedLinearAccClamp=<3000 3000 3000>
    componentLinearAccScale=<1 1 1>
    componentLinearVelScale=<1 1 1>
    gravityOverride {z=-980}
    physicsBodyDefinitions [{
      boundBone {boneName=dyn_hair_mid_2}
      boxExtents=<20 30 30>
      constraintSetup {angularLimitsMax=<0 0 5>, angularLimitsMin=<0 -15 -5>, angularTarget=<1 0 0>}
      localJointOffset=<4 0 0>
     }]
   }
   n18 = Anim Dynamics - Bone: dyn_hair_mid_3!(ComponentPose=@n17, bAlphaBoolEnabled:bool=True, GravityOverride:Vector="0,0,-980", ExternalForce="0,0,0") -> (Pose) {
    alphaInputType=Bool
    angularDampingOverride=1
    angularSpringConstant=10000
    bAngularSpring=true
    bOverrideAngularDamping=true
    bUseGravityOverride=true
    gravityScale=0.35
    simulationSpace=RootRelative
    boundBone {boneName=dyn_hair_mid_3}
    componentAppliedLinearAccClamp=<3000 3000 3000>
    componentLinearAccScale=<1 1 1>
    componentLinearVelScale=<1 1 1>
    gravityOverride {z=-980}
    physicsBodyDefinitions [{
      boundBone {boneName=dyn_hair_mid_3}
      boxExtents=<20 30 30>
      constraintSetup {angularLimitsMax=<0 0 5>, angularLimitsMin=<0 -15 -5>, angularTarget=<1 0 0>}
      localJointOffset=<4 0 0>
     }]
   }
   n19 = Anim Dynamics - Bone: dyn_hair_mid_4!(ComponentPose=@n18, bAlphaBoolEnabled:bool=True, GravityOverride:Vector="0,0,-980", ExternalForce="0,0,0") -> (Pose) {
    alphaInputType=Bool
    angularDampingOverride=1
    angularSpringConstant=10000
    bAngularSpring=true
    bOverrideAngularDamping=true
    bUseGravityOverride=true
    gravityScale=0.35
    simulationSpace=RootRelative
    boundBone {boneName=dyn_hair_mid_4}
    componentAppliedLinearAccClamp=<3000 3000 3000>
    componentLinearAccScale=<1 1 1>
    componentLinearVelScale=<1 1 1>
    gravityOverride {z=-980}
    physicsBodyDefinitions [{
      boundBone {boneName=dyn_hair_mid_4}
      boxExtents=<20 30 30>
      constraintSetup {angularLimitsMax=<0 0 5>, angularLimitsMin=<0 -15 -5>, angularTarget=<1 0 0>}
      localJointOffset=<4 0 0>
     }]
   }
   n20 = "Transform (Modify) Bone - Bone: dyn_hair_back_1_r"(ComponentPose=@n19, Alpha=1, Translation="0,0,0", Rotation="-15.149,-27.450,9.074", Scale="0.857,1.000,1.000") {
    rotationMode=BMM_Additive
    rotationSpace=BCS_BoneSpace
    scaleMode=BMM_Additive
    scaleSpace=BCS_BoneSpace
    translationMode=BMM_Additive
    translationSpace=BCS_BoneSpace
    alphaScaleBiasClamp {bInterpResult=true, interpSpeedDecreasing=0.5, interpSpeedIncreasing=0.5}
    boneToModify {boneName=dyn_hair_back_1_r}
    rotation=<rot -15.149 -27.45 9.074>
    scale {x=0.857}
   }
   n21 = "Transform (Modify) Bone - Bone: dyn_hair_back_1_l"(ComponentPose=@n20, Alpha=1, Translation="-2.480,-0.531,-4.002", Rotation="-15.157,14.856,-3.572", Scale="1,1,1") {
    rotationMode=BMM_Additive
    rotationSpace=BCS_BoneSpace
    scaleMode=BMM_Additive
    scaleSpace=BCS_BoneSpace
    translationMode=BMM_Additive
    translationSpace=BCS_BoneSpace
    alphaScaleBiasClamp {bInterpResult=true, interpSpeedDecreasing=0.5, interpSpeedIncreasing=0.5}
    boneToModify {boneName=dyn_hair_back_1_l}
    rotation=<rot -15.157 14.856 -3.572>
    translation=<-2.48 -0.531 -4.002>
   }
   n22 = "Transform (Modify) Bone - Bone: dyn_hair_mid_1"(ComponentPose=@n21, Alpha=1, Translation="0,0,0", Rotation="12.996,-20.280,-8.015", Scale="0.742,1.000,1.000") {
    rotationMode=BMM_Additive
    rotationSpace=BCS_BoneSpace
    scaleMode=BMM_Additive
    scaleSpace=BCS_BoneSpace
    translationMode=BMM_Additive
    translationSpace=BCS_BoneSpace
    alphaScaleBiasClamp {bInterpResult=true, interpSpeedDecreasing=0.5, interpSpeedIncreasing=0.5}
    boneToModify {boneName=dyn_hair_mid_1}
    rotation=<rot 12.996 -20.28 -8.015>
    scale {x=0.742}
   }
   n23 = "Constraint - dyn_hair_back_1_l (PP)"!(ComponentPose=@n22, bAlphaBoolEnabled:bool=True, ConstraintWeights_0=0.8, ConstraintWeights_1:float=0.2) -> (Pose) {
    alphaInputType=Bool
    boneToModify {boneName=dyn_hair_back_1_l}
    constraintSetup [{targetBone {boneName=head}}, {targetBone {boneName=spine_05}}]
    constraintWeights [1,0]
   }
   n24 = "Constraint - dyn_hair_mid_1 (PP)"!(ComponentPose=@n23, bAlphaBoolEnabled:bool=True, ConstraintWeights_0=0.8, ConstraintWeights_1:float=0.2) -> (Pose) {
    alphaInputType=Bool
    boneToModify {boneName=dyn_hair_mid_1}
    constraintSetup [{targetBone {boneName=head}}, {targetBone {boneName=spine_05}}]
    constraintWeights [1,0]
   }
   n25 = "Constraint - dyn_hair_back_1_r (PP)"!(ComponentPose=@n24, bAlphaBoolEnabled:bool=True, ConstraintWeights_0=0.8, ConstraintWeights_1:float=0.2) -> (Pose) {
    alphaInputType=Bool
    boneToModify {boneName=dyn_hair_back_1_r}
    constraintSetup [{targetBone {boneName=head}}, {targetBone {boneName=spine_05}}]
    constraintWeights [1,0]
   }
   n26 = "Constraint - dyn_hair_back_2_l (PP)"!(ComponentPose=@n25, bAlphaBoolEnabled:bool=True, ConstraintWeights_0=0.3, ConstraintWeights_1:float=0.7) -> (Pose) {
    alphaInputType=Bool
    boneToModify {boneName=dyn_hair_back_2_l}
    constraintSetup [{targetBone {boneName=head}}, {targetBone {boneName=clavicle_l}}]
    constraintWeights [1,1]
   }
   n27 = "Constraint - dyn_hair_mid_2 (PP)"!(ComponentPose=@n26, bAlphaBoolEnabled:bool=True, ConstraintWeights_0=0.3, ConstraintWeights_1:float=0.7) -> (Pose) {
    alphaInputType=Bool
    boneToModify {boneName=dyn_hair_mid_2}
    constraintSetup [{targetBone {boneName=head}}, {targetBone {boneName=spine_05}}]
    constraintWeights [1,1]
   }
   n28 = "Constraint - dyn_hair_back_2_r (PPP)"!(ComponentPose=@n27, bAlphaBoolEnabled:bool=True, ConstraintWeights_0=0.15, ConstraintWeights_1:float=0.7, ConstraintWeights_2:float=0.15) -> (Pose) {
    alphaInputType=Bool
    boneToModify {boneName=dyn_hair_back_2_r}
    constraintSetup [{targetBone {boneName=head}}, {targetBone {boneName=clavicle_r}}, {targetBone {boneName=spine_05}}]
    constraintWeights [1,1,1]
   }
   n29 = "Constraint - dyn_hair_back_3_l (P)"!(ComponentPose=@n28, bAlphaBoolEnabled:bool=True, ConstraintWeights_0=1) -> (Pose) {alphaInputType=Bool, boneToModify {boneName=dyn_hair_back_3_l}, constraintSetup [{targetBone {boneName=clavicle_l}}], constraintWeights [1]}
   n30 = "Constraint - dyn_hair_mid_3 (P)"!(ComponentPose=@n29, bAlphaBoolEnabled:bool=True, ConstraintWeights_0=1) -> (Pose) {alphaInputType=Bool, boneToModify {boneName=dyn_hair_mid_3}, constraintSetup [{targetBone {boneName=spine_05}}], constraintWeights [1]}
   n31 = "Constraint - dyn_hair_back_3_r (P)"!(ComponentPose=@n30, bAlphaBoolEnabled:bool=True, ConstraintWeights_0=1) -> (Pose) {alphaInputType=Bool, boneToModify {boneName=dyn_hair_back_3_r}, constraintSetup [{targetBone {boneName=clavicle_r}}], constraintWeights [1]}
   n32 = Control Rig!(
    Source=Component To Local(ComponentPose=@n31)
    Alpha
    UsingPistol=Get Pistol Equipped_Float()
    IsTargeting=Get Istargeting_Float()
    not in backbling:bool=Get IsWearingBackbling() -> (IsWearingBackbling:bool)
    RBANOn:bool=Get bRigidBodyEnabled()
   ) -> (Pose) {controlRigAssetReference {blueprintRigClass=&/F_MED_GoatFish_FaceAcc_PostCtrlRig.F_MED_GoatFish_FaceAcc_PostCtrlRig_C}}
   n33 = "Transform (Modify) Bone - Bone: dyn_hair_back_1_l"!(
    ComponentPose=Local To Component(LocalPose=@n32)
    bAlphaBoolEnabled:bool=True
    Translation="0,0,0"
    Rotation="10.703,-0.081,0.005"
    Scale="1,1,1"
   ) -> (Pose) {
    alphaInputType=Bool
    rotationMode=BMM_Additive
    rotationSpace=BCS_BoneSpace
    scaleMode=BMM_Additive
    scaleSpace=BCS_BoneSpace
    translationMode=BMM_Additive
    translationSpace=BCS_BoneSpace
    alphaScaleBiasClamp {bInterpResult=true, interpSpeedDecreasing=0.5, interpSpeedIncreasing=0.5}
    boneToModify {boneName=dyn_hair_back_1_l}
    rotation=<rot 10.703 -0.081 0.005>
   }
   n34 = "Transform (Modify) Bone - Bone: dyn_hair_mid_1"!(ComponentPose=@n33, bAlphaBoolEnabled:bool=True, Translation="0,0,0", Rotation="-0.282,-0.106,-0.007", Scale="1,1,1") -> (Pose) {
    alphaInputType=Bool
    rotationMode=BMM_Additive
    rotationSpace=BCS_BoneSpace
    scaleMode=BMM_Additive
    scaleSpace=BCS_BoneSpace
    translationMode=BMM_Additive
    translationSpace=BCS_BoneSpace
    alphaScaleBiasClamp {bInterpResult=true, interpSpeedDecreasing=0.5, interpSpeedIncreasing=0.5}
    boneToModify {boneName=dyn_hair_mid_1}
    rotation=<rot -0.282 -0.106 -0.007>
   }
   n35 = "Transform (Modify) Bone - Bone: dyn_hair_back_1_r"!(ComponentPose=@n34, bAlphaBoolEnabled:bool=True, Translation="0,0,0", Rotation="1.998,0.066,-0.132", Scale="1,1,1") -> (Pose) {
    alphaInputType=Bool
    rotationMode=BMM_Additive
    rotationSpace=BCS_BoneSpace
    scaleMode=BMM_Additive
    scaleSpace=BCS_BoneSpace
    translationMode=BMM_Additive
    translationSpace=BCS_BoneSpace
    alphaScaleBiasClamp {bInterpResult=true, interpSpeedDecreasing=0.5, interpSpeedIncreasing=0.5}
    boneToModify {boneName=dyn_hair_back_1_r}
    rotation=<rot 1.998 0.066 -0.132>
   }
   Output Pose(Result=Component To Local(ComponentPose=@n35)) {layerGroup=None}
  ]}
 EventGraph {nodes [
   Event BlueprintUpdateAnimation()
   > n1 = Sequence!(execute) -> (then_0, then_1, then_2:Exec, then_3:Exec, then_4:Exec, then_5:Exec, then_6:Exec, then_7:Exec, then_8:Exec)
   > n2 = "Gravity Override State Function (Parameter Asset) (Nativized)"(
    CharacterPartAnimInstance=Self-Reference()
    GameTimeInSeconds=GetGameTimeInSeconds()
    Parameters=/BRCosmetics/Characters/Player/Female/Medium/Bodies/F_MED_GoatFish/Meshes/F_MED_GoatFish_Parameters
   )
   > Set Gravity Override(Gravity Override=@n2)
   > Set RBAN Settings Out(RBAN Settings Out=@n2.OutRigidBodyAnimNodeParameters)
   Set Crouchwalk(execute=@n1.then_3, Crouchwalk:bool=AND Boolean(A=Get bIsCrouching(), B=Get bIsMoving2D())) -> (Output_Get:bool)
   Set CrouchHeadUP(execute=@n1.then_4, CrouchHeadUP:double=SelectFloat(A=Get Head Up() -> (Head Up:double), B=0, bPickA=Get Crouchwalk())) -> (Output_Get:double)
   n3 = Get Crouchwalk()
   > Set CrouchNOt walk(CrouchNOt walk:bool=AND Boolean(A=Get bIsCrouching(), B=NOT Boolean(A=@n3))) -> (Output_Get:bool)
   n4 = Get bIsCrouching()
   n5 = Get bIsTargeting()
   > Set TargetingHeadUP(TargetingHeadUP:double=SelectFloat(A=Get Head Up(), B=0, bPickA=AND Boolean(A=@n5, B=@n4))) -> (Output_Get:double)
   > n6 = Branch(Condition=Get bIsCrouching())
   > Set RigidOFF(RigidOFF:double=0) -> (Output_Get:double)
   n7 = Set RigidOFF(execute=@n6.else, RigidOFF=1)
   > Set RBANOn(RBANOn:double=@n7) -> (Output_Get:double)
   > Set RBANOn_0(RBANOn_0:bool=false) -> (Output_Get:bool)
   n8 = NOT Boolean(A=Get bIsMoving2D())
   n9 = Get bIsTargeting()
   n10 = Get bIsCrouching()
   n11 = Get Pistol Equipped() -> (Pistol Equipped:bool)
   Set PistolHeadUP(
    execute=@n1.then_6
    PistolHeadUP:double=SelectFloat(A=Get Head Up(), B=0, bPickA=AND Boolean!(A=@n11, B=@n10, C:bool=@n9, D:bool=@n8) -> (ReturnValue))
   ) -> (Output_Get:double)
   n12 = Get Pistol Equipped()
   n13 = Get bIsTargeting()
   > Set StandingPistolHeadUP(StandingPistolHeadUP:double=SelectFloat(A=Get Head Up(), B=0, bPickA=AND Boolean!(A=@n12, B=@n13) -> (ReturnValue))) -> (Output_Get:double)
   Set Pistol Equipped_Float(execute=@n1.then_7, Pistol Equipped_Float:double="To Float (Boolean)"(InBool=Get Pistol Equipped())) -> (Output_Get:double)
   > Set Istargeting_Float(Istargeting_Float:double="To Float (Boolean)"(InBool=Get bIsTargeting())) -> (Output_Get:double)
   n14 = Get bIsPlayingEmote()
   Set FEIdle(execute=@n1.then_8, FEIdle:bool=AND Boolean(A=Get bIsInFrontEnd(), B=NOT Boolean(A=@n14))) -> (Output_Get:bool)
   n15 = Get Skeletal_Mesh()
   Set Head Up(
    execute=@n1.then_2
    Head Up:double=GetAlphaFromReferenceTransformDelta(SkeletalMeshComponent=@n15, BoneName=spine_05, BaseName=head, InputRange=(X=15,Y=70), TwistAxis="0,0,1")
   ) -> (Output_Get:double)
   > Set Head Down(Head Down:double=GetAlphaFromReferenceTransformDelta(SkeletalMeshComponent=@n15, BoneName=spine_05, BaseName="head ", InputRange=(X=0,Y=-90), OutputRange=(X=0,Y=-1), TwistAxis="0,0,1")) -> (Output_Get:double)
   > Set Head_Turn_L(Head_Turn_L:double=GetAlphaFromReferenceTransformDelta(SkeletalMeshComponent=@n15, BoneName=spine_05, BaseName=head, DeltaComponentType=EulerX, InputRange=(X=0,Y=-22), TwistAxis="1,0,0")) -> (Output_Get:double)
   > Set Head_Turn_R(Head_Turn_R:double=GetAlphaFromReferenceTransformDelta(SkeletalMeshComponent=@n15, BoneName=spine_05, BaseName=head, DeltaComponentType=EulerX, InputRange=(X=0,Y=22), TwistAxis="1,0,0")) -> (Output_Get:double)
   n16 = CheckForBackbling()
   Set IsWearingBackbling(execute=@n1.then_1, IsWearingBackbling:bool=@n16) -> (Output_Get:bool)
   > Set Is Backbling A Cape(Is Backbling A Cape:bool=@n16.bIsBackblingACape) -> (Output_Get:bool)
   > Set is Wear a Cape(is Wear a Cape:bool=AND Boolean(A=@n16, B=@n16.bIsBackblingLong)) -> (Output_Get:bool)
   n17 = Cast To FortPlayerAnimInstance(execute=@n1.then_5, Object=GetAnimInstance(self=GetOwnerSkeletalMeshComponent()))
   > Set LongW Equipped(LongW Equipped:bool=false) -> (Output_Get:bool)
   > n18 = Switch on EFortWeaponCoreAnimation!(execute, Selection:EFortWeaponCoreAnimation Enum=Get WeaponCoreAnim(self=@n17)) -> (
    Melee:Exec
    Pistol:Exec
    Shotgun:Exec
    PaperBlueprint:Exec
    Rifle:Exec
    MeleeOneHand:Exec
    MachinePistol:Exec
    RocketLauncher:Exec
    GrenadeLauncher:Exec
    GoingCommando:Exec
    AssaultRifle:Exec
    TacticalShotgun:Exec
    SniperRifle:Exec
    TrapPlacement:Exec
    ShoulderLauncher:Exec
    AbilityDecoTool:Exec
    Crossbow:Exec
    C4:Exec
    RemoteControl:Exec
    DualWield:Exec
    AR_BullPup:Exec
    AR_ForwardGrip:Exec
    MedPackPaddles:Exec
    SMG_P90:Exec
    AR_DrumGun:Exec
    Consumable_Small:Exec
    Consumable_Large:Exec
    Balloon:Exec
    MountedTurret:Exec
    CreativeTool:Exec
    ExplosiveBow:Exec
    AshtonIndigo:Exec
    AshtonChicago:Exec
    MeleeDualWield:Exec
    Unarmed:Exec
   )
   n19 = Branch(execute=[@n18.DualWield, @n18.Pistol], Condition=Get bIsTargeting())
   > Set Pistol Equipped(Pistol Equipped:bool=true) -> (Output_Get:bool)
   Set Pistol Equipped(execute=@n19.else, Pistol Equipped=false)
   Event BlueprintInitializeAnimation()
   > Set Skeletal_Mesh(Skeletal_Mesh:Skeletal Mesh Component Object Reference=GetOwnerSkeletalMeshComponent()) -> (Output_Get:Skeletal Mesh Component Object Reference)
  ]}
]
variables [
 CrouchHeadUP {type=number}
 Crouching ADS UP {type=number}
 CrouchNOt walk {type=boolean}
 Crouchwalk {type=boolean}
 CrouchwalkHeadUP {type=number}
 FEIdle {type=boolean}
 GoingUP {type=boolean}
 Gravity Override {type=Vector}
 Head Down {type=number}
 Head look up alpha {type=number}
 Head Up {type=number}
 Head_Turn_L {type=number}
 Head_Turn_R {type=number}
 Is Backbling A Cape {type=boolean}
 is Wear a Cape {type=boolean}
 Istargeting_Float {type=number}
 isUsingPistol {type=number}
 IsWearingBackbling {type=boolean}
 LongW Equipped {type=boolean}
 Pistol Equipped {type=boolean}
 Pistol Equipped_Float {type=number}
 PistolHeadUP {type=number}
 RBAN Settings Out {type=FortRigidBodyAnimNodeParameters}
 RBANOn {type=number}
 RBANOn_0 {type=boolean}
 RigidOFF {type=number}
 Skel Mesh {type=SkeletalMeshComponent}
 Skeletal_Mesh {type=SkeletalMeshComponent}
 StandingPistolHeadUP {type=number}
 TargetingHeadUP {type=number}
]
